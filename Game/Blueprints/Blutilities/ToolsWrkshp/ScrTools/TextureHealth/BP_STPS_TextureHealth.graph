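# BP_STPS_TextureHealth is an Editor Utility Blueprint that holds options and event dispatchers for a texture health tool.
# Run Tests and Next Texture only broadcast their matching dispatchers, leaving the actual work to listeners outside this asset.
# A clear-collections dispatcher is also declared, but nothing here calls it.
# Variables choose which problem collection to cycle through, defaulting to Severe, and whether to open the asset editor.
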EditorUtilityBlueprint /Game/Blueprints/Blutilities/ToolsWrkshp/ScrTools/TextureHealth/BP_STPS_TextureHealth
graphs [
 EventGraph {links [], nodes []}
 "Next Texture " {nodes [
   "Next Texture "() {is=title K2Node_FunctionEntry}
   > Call On Step to Next Problem Texture Clicked!(execute:Exec, self:BP STPS Texture Health Object Reference) -> (then:Exec) {is=title K2Node_CallDelegate}
  ]}
 OnClearTextureCollectionsClicked {nodes [OnClearTextureCollectionsClicked() {is=title K2Node_FunctionEntry}]}
 OnRunTestsClicked {nodes [OnRunTestsClicked() {is=title K2Node_FunctionEntry}]}
 OnStepToNextProblemTextureClicked {nodes [OnStepToNextProblemTextureClicked() {is=title K2Node_FunctionEntry}]}
 Run Tests {nodes [
   Run Tests() {is=title K2Node_FunctionEntry}
   > Call On Run Tests Clicked!(execute:Exec, self:BP STPS Texture Health Object Reference) -> (then:Exec) {is=title K2Node_CallDelegate}
  ]}
]
variables [
 Assets to Cycle Through {type=E_CollectionChoices_TextureHealth, default=Severe}
 OnClearTextureCollectionsClicked {default=()}
 OnRunTestsClicked {default=()}
 OnStepToNextProblemTextureClicked {default=()}
 OpenAssetEditor {type=boolean}
]
