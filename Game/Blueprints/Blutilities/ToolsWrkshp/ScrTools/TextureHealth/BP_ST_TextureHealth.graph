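# Editor scriptable tool that audits the health of Texture2D assets in the main Valkyrie project.
# Run Tests checks size, power-of-two, streaming and compression issues, then merges results into one map per texture.
# Error count sorts textures into Severe, Moderate and Minor, saved as colored Content Browser collections.
# Next Texture cycles the chosen collection, syncing the browser and rebuilding the header with Fix Me buttons for oversized and non-power-of-two textures.
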EditorUtilityBlueprint /Game/Blueprints/Blutilities/ToolsWrkshp/ScrTools/TextureHealth/BP_ST_TextureHealth
graphs [
 AddActionsBox {nodes [
   AddActionsBox() {is=title K2Node_FunctionEntry}
   > n1 = AddExpandableBox(self=Get Header() -> (Header:BP UW Shared Header Widget Object Reference))
   > n2 = Set Actions Box(Actions Box=@n1)
   > SetUpBox(self=@n2, Name=Get Actions Box Label(self=Get Header()))
  ]}
 AddNextTextureButton {nodes [
   AddNextTextureButton() {is=title K2Node_FunctionEntry}
   > AddChildToVerticalBox(
    self=Get DynamicEntries(self=Get Actions Box())
    Content=Get NextTextureButton() -> (NextTextureButton:BP UW Next Texture Button Object Reference)
   )
   > Bind Event to Next Texture Clicked(self:BP UW Next Texture Button Object Reference=Get NextTextureButton(), Delegate=Create Event()) {is=title K2Node_AddDelegate}
  ]}
 AddRunTestsButton {nodes [
   AddRunTestsButton() {is=title K2Node_FunctionEntry}
   > AddChildToVerticalBox(self=Get DynamicEntries(self=Get Actions Box()), Content=Get RunTestsButton() -> (RunTestsButton:BP UW Run Tests Button Object Reference))
   > Bind Event to Run Tests Clicked(self:BP UW Run Tests Button Object Reference=Get RunTestsButton(), Delegate=Create Event()) {is=title K2Node_AddDelegate}
  ]}
 CreateCollectionMessage {nodes [
   n1 = CreateCollectionMessage!() -> (then, CurSev:E_CollectionChoices_TextureHealth Enum=NewEnumerator0, NumFound:int) {is=title K2Node_FunctionEntry}
   n2 = "To String (Text)"(InText=Get FinalRunMessage() -> (FinalRunMessage:Text))
   > n3 = Branch(Condition="integer > integer"(A=Len(S=@n2)))
   n4 = SelectString(A=MakeLiteralString(), B=MakeLiteralString(Value=s), bPickA="Equal (Integer)"(A=@n1.NumFound, B=1))
   n5 = Enum to String!(Enumerator:E_CollectionChoices_TextureHealth Enum=@n1) -> (ReturnValue:String)
   n6 = SelectString(
    A="To String (Text)"(InText=Format Text!(Format="LOCTEXT(\"Found {num} {sev} asset{ss}, no collection created\")", num:int=@n1.NumFound, sev:String=@n5, ss:String=@n4) -> (Result))
    B="To String (Text)"(InText=Format Text(Format="LOCTEXT(\"Found {num} {sev} asset{ss}, added to collection \\\"{sev}\\\"\")", num=@n1.NumFound, sev=@n5, ss=@n4))
    bPickA="integer < integer"(A=@n1.NumFound, B=1)
   )
   > Set FinalRunMessage(FinalRunMessage:Text=Format Text!(Format="LOCTEXT(\"{p}\\r\\n{m}\")", p:Text=Get FinalRunMessage(), m:String=@n6) -> (Result)) -> (Output_Get:Text)
   Set FinalRunMessage(execute=@n3.else, FinalRunMessage="To Text (String)"(InString=@n6))
  ]}
 EventGraph {nodes [
   n1 = Sync Browser!() -> (OutputDelegate, then, Object:Object Reference) {is=title K2Node_CustomEvent}
   > Delay(Duration=0.2)
   > SyncBrowserToObjects(AssetPaths=Make Array("[0]":String=GetSoftObjectPath!(Object=@n1.Object) -> (ReturnValue_PathString:String)) -> (Array:Array of Strings))
   Event OnScriptSetup()
   > n2 = Sequence!(execute) -> (then_0, then_1, then_2:Exec, then_3:Exec, then_4:Exec, then_5:Exec, then_6:Exec, then_7:Exec, then_8:Exec)
   > Set Cur Texture Selected(Cur Texture Selected:int=0) -> (Output_Get:int=0)
   K2Node_Composite_0 = SetupPropertySets(execute=@n2.then_1) {is=title K2Node_Composite}
   K2Node_Composite_1 = AddHeaderWidget(execute=@n2.then_2) {is=title K2Node_Composite}
   K2Node_Composite_2 = SetupPropertyCallbacks(execute=@n2.then_3) {is=title K2Node_Composite}
   K2Node_Composite_3 = InitProjectVariables(execute=@n2.then_4) {is=title K2Node_Composite}
   SetFocusInViewport(execute=@n2.then_5, self:Scriptable Interactive Tool Object Reference)
   K2Node_Composite_5 = InitCollectionColors(execute=@n2.then_6) {is=title K2Node_Composite}
   Populate Collection Vars(execute=@n2.then_7)
   DisplayUserWarningMessage(execute=@n2.then_8, Message="LOCTEXT(\"\\r\\n\")")
   Event OnScriptShutdown()
   > SavePropertySetSettings(
    PropertySet=Get Settings() -> (Settings:BP STPS Texture Health Object Reference)
    SaveKey=Get Prop Set Save Key() -> (Prop Set Save Key:String)
   )
   Event OnScriptTick()
   > Set CurTime(CurTime=float + float(B=Get CurTime()))
   > n3 = Branch(Condition=Get FirstInitRun())
   Set FirstInitRun(execute=@n3.else)
   > AddActionsBox() {is=title K2Node_CallFunction}
   > n4 = Sequence!(execute) -> (then_0, then_1, then_2)
   > n5 = Construct Object from Class(Class=/Game/Blueprints/Blutilities/ToolsWrkshp/ScrTools/TextureHealth/SupportFiles/BP_UW_RunTestsButton.BP_UW_RunTestsButton_C) -> (ReturnValue:BP UW Run Tests Button Object Reference)
   > Set RunTestsButton(RunTestsButton:BP UW Run Tests Button Object Reference=@n5) -> (Output_Get:BP UW Run Tests Button Object Reference)
   > AddRunTestsButton() {is=title K2Node_CallFunction}
   n6 = Construct Object from Class(
    execute=@n4.then_1
    Class=/Game/Blueprints/Blutilities/ToolsWrkshp/ScrTools/TextureHealth/SupportFiles/BP_UW_NextTextureButton.BP_UW_NextTextureButton_C
   ) -> (ReturnValue:BP UW Next Texture Button Object Reference)
   > Set NextTextureButton(NextTextureButton:BP UW Next Texture Button Object Reference=@n6) -> (Output_Get:BP UW Next Texture Button Object Reference)
   > NextButtonEnable!(execute, self, Enable:bool=false) -> (then)
   > AddNextTextureButton() {is=title K2Node_CallFunction}
   K2Node_Composite_4 = InitCollectionNames!(InputPin:Exec) {is=title K2Node_Composite}
  ]}
 SetupPropertySets {
  path=EventGraph.K2Node_Composite_0.SetupPropertySets
  nodes [
   Inputs()
   > n1 = AddPropertySetOfType(
    PropertySetType=&/BP_STPS_TextureHealth.BP_STPS_TextureHealth_C
    Identifier=Settings
   )
   > n2 = Cast To BP_STPS_TextureHealth!(execute, Object=@n1) -> (then, CastFailed, AsBP STPS Texture Health:BP STPS Texture Health Object Reference) {is=title K2Node_DynamicCast}
   > n3 = Set Settings(Settings:BP STPS Texture Health Object Reference=@n2) -> (Output_Get:BP STPS Texture Health Object Reference)
   > RestorePropertySetSettings(PropertySet=@n3, SaveKey=Get Prop Set Save Key())
   Outputs()
  ]
 }
 AddHeaderWidget {
  path=EventGraph.K2Node_Composite_1.AddHeaderWidget
  nodes [
   Inputs()
   > n1 = Construct Object from Class(Class=&BP_UW_SharedHeaderWidget_C) -> (ReturnValue:BP UW Shared Header Widget Object Reference)
   > n2 = Set Header(Header:BP UW Shared Header Widget Object Reference=@n1) -> (Output_Get:BP UW Shared Header Widget Object Reference)
   > SetToolPanelToolHeaderWidget(Widget=@n2)
   Outputs()
  ]
 }
 SetupPropertyCallbacks {
  path=EventGraph.K2Node_Composite_2.SetupPropertyCallbacks
  nodes [
   Inputs()
   > Bind Event to On Run Tests Clicked(self:BP STPS Texture Health Object Reference=Get Settings(), Delegate=Create Event()) {is=title K2Node_AddDelegate}
   > Bind Event to On Step to Next Problem Texture Clicked(self:BP STPS Texture Health Object Reference=Get Settings(), Delegate=Create Event()) {is=title K2Node_AddDelegate}
   > WatchProperty(PropertySet=Get Settings(), PropertyName=Assets to Cycle Through, OnModified=Create Event())
   Outputs()
  ]
 }
 InitProjectVariables {
  path=EventGraph.K2Node_Composite_3.InitProjectVariables
  nodes [
   Inputs()
   > n1 = GetMainProject!(execute) -> (then, ReturnValue:Valkyrie Project Handle Structure)
   > n2 = Set Main Project Handle(Main Project Handle:Valkyrie Project Handle Structure=@n1) -> (Output_Get:Valkyrie Project Handle Structure)
   > Set Project Handle Valid(Project Handle Valid:bool=IsProjectHandleValid!(ProjectHandle:"Valkyrie Project Handle Structure (by ref)"=@n2) -> (ReturnValue:bool)) -> (Output_Get:bool)
   > Set Project Asset Path(Project Asset Path:String=GetProjectMountedAssetPath!(ProjectHandle:"Valkyrie Project Handle Structure (by ref)"=@n2) -> (ReturnValue:String)) -> (Output_Get:String)
   > Set Valkyrie Main Proj Collection Container(Valkyrie Main Proj Collection Container:Collection Container Source Structure=GetProjectCollectionContainer!(ProjectHandle:"Valkyrie Project Handle Structure (by ref)"=Get Main Project Handle() -> (Main Project Handle:Valkyrie Project Handle Structure)) -> (OutProjectCollectionContainer:Collection Container Source Structure, ReturnValue:bool)) -> (Output_Get:Collection Container Source Structure)
   Outputs()
  ]
 }
 InitCollectionNames {
  path=EventGraph.K2Node_Composite_4.InitCollectionNames
  nodes [
   n1 = Inputs!() -> (InputPin:Exec)
   Set Moderate Collection Name(
    execute=[@n2.then, @n1.InputPin]
    Moderate Collection Name:String=Append!(
     A=Get Parent Collection Name() -> (Parent Collection Name:String)
     B=_
     C:String=Get Moderate Collection Name() -> (Moderate Collection Name:String)
    ) -> (ReturnValue)
   ) -> (Output_Get:String)
   > Set Minor Collection Name(Minor Collection Name:String=Append(A=Get Parent Collection Name(), B=_, C=Get Minor Collection Name() -> (Minor Collection Name:String))) -> (Output_Get:String)
   n2 = Set Severe Collection Name(Severe Collection Name:String=Append(A=Get Parent Collection Name(), B=_, C=Get Severe Collection Name() -> (Severe Collection Name:String))) -> (Output_Get:String)
   Outputs()
  ]
 }
 InitCollectionColors {
  path=EventGraph.K2Node_Composite_5.InitCollectionColors
  nodes [
   Inputs!() -> (execute:Exec)
   > n1 = Get Collection Colors!(execute) -> (
    then
    Common Parent Color:Linear Color Structure
    "Common  Severe Collection Color":Linear Color Structure
    Common Moderate Collection Color:Linear Color Structure
    Common Minor Collection Color:Linear Color Structure
   ) {is=title K2Node_CallFunction}
   > Set Parent Color(Parent Color:Linear Color Structure=@n1) -> (Output_Get:Linear Color Structure)
   > Set Severe Collection Color(Severe Collection Color:Linear Color Structure=@n1."Common  Severe Collection Color") -> (Output_Get:Linear Color Structure)
   > Set Moderate Collection Color(Moderate Collection Color:Linear Color Structure=@n1.Common Moderate Collection Color) -> (Output_Get:Linear Color Structure)
   > Set Minor Collection Color(Minor Collection Color:Linear Color Structure=@n1.Common Minor Collection Color) -> (Output_Get:Linear Color Structure)
   Outputs()
  ]
 }
 NextButtonEnable {nodes [
   n1 = NextButtonEnable!() -> (then, Enable:bool=true) {is=title K2Node_FunctionEntry}
   > SetIsEnabled(self=Get NextTextureButton(), bInIsEnabled=@n1)
   n2 = SelectText(
    A=Get NextButtonTooltip_Enabled() -> (NextButtonTooltip_Enabled:Text)
    B=Get NextButtonTooltip_Disabled() -> (NextButtonTooltip_Disabled:Text)
    bPickA=@n1
   )
   n3 = Get NextTextureButton()
   > SetToolTipText!(execute, self:Widget Object Reference=@n3, InToolTipText:"Text (by ref)"=@n2) -> (then)
   > Set ToolTipText!(execute, ToolTipText:Text=@n2, self:Widget Object Reference=@n3) -> (then, Output_Get:Text)
  ]}
 On Next Prob Texture Clicked {nodes [
   On Next Prob Texture Clicked() {is=title K2Node_FunctionEntry}
   n1 = Get Assets to Cycle Through!(self:BP STPS Texture Health Object Reference=Get Settings()) -> (Assets to Cycle Through:E_CollectionChoices_TextureHealth Enum=NewEnumerator0)
   n2 = Select!(
    NewEnumerator0:Collection Structure=Get Severe Collection() -> (Severe Collection:Collection Structure)
    NewEnumerator1:Collection Structure=Get Moderate Collection() -> (Moderate Collection:Collection Structure)
    NewEnumerator2:Collection Structure=Get Minor Collection() -> (Minor Collection:Collection Structure)
    Index:E_CollectionChoices_TextureHealth Enum=@n1
   ) -> (ReturnValue:Collection Structure)
   > n3 = CollectionExists!(
    execute
    self:Collection Manager Subsystem Object Reference=Get CollectionManagerScriptingSubsystem()
    Container:Collection Container Source Structure=Get Valkyrie Main Proj Collection Container() -> (Valkyrie Main Proj Collection Container:Collection Container Source Structure)
    Collection:Name=Break Collection!(CollectionScriptingRef:"Collection Structure (by ref)"=@n2) -> (Container:Name, Name:Name, ShareType:ECollectionScriptingShareType Enum) {is=title K2Node_BreakStruct}.Name
    ShareType:ECollectionScriptingShareType Enum=Get Collection Share Type() -> (Collection Share Type:ECollectionScriptingShareType Enum=Local)
   ) -> (then, ReturnValue:bool)
   > n4 = Branch(Condition=@n3)
   > n5 = GetAssetsInCollection!(
    execute
    self:Collection Manager Subsystem Object Reference=Get CollectionManagerScriptingSubsystem()
    Collection:"Collection Structure (by ref)"=@n2
   ) -> (then, OutAssets:Array of Asset Data Structures, ReturnValue:bool)
   n6 = Set Num Assets In This Collection(execute=[@n5.then, @n26.then], Num Assets In This Collection:int=Length(TargetArray:Array of Asset Data Structures=@n5)) -> (Output_Get:int=0)
   > n7 = Branch(Condition=OR Boolean(A=NOT Boolean(A=@n5.ReturnValue), B="integer < integer"(A=@n6, B=1)))
   > DisplayUserWarningMessage(Message=Format Text!(Format="LOCTEXT(\"No {s} assets found\")", s:String=Enum to String(Enumerator=@n1)) -> (Result))
   DisplayUserWarningMessage(execute=@n7.else)
   > Branch(Condition=@n5.ReturnValue)
   n8 = Get Cur Texture Selected() -> (Cur Texture Selected:int=0)
   > n9 = Cast To Texture2D(Object=GetAsset(InAssetData="Get (a ref)"(Array:Array of Asset Data Structures=@n5, Dimension 1=@n8) -> (Output:"Asset Data Structure (by ref)")))
   > n10 = Set Cur Texture(Cur Texture:Texture 2D Object Reference=@n9) -> (Output_Get:Texture 2D Object Reference)
   > n11 = Branch(Condition=Find(
     TargetMap:Map of Texture 2D Object References to S Asset Errors Asset Health Structures=Get All Textures() -> (All Textures:Map of Texture 2D Object References to S Asset Errors Asset Health Structures)
     Key:"Texture 2D Object Reference (by ref)"=Get Cur Texture() -> (Cur Texture:Texture 2D Object Reference)
    ) -> (Value:S Asset Errors Asset Health Structure).ReturnValue)
   > n12 = Set Currently Selected Asset(Currently Selected Asset:Texture 2D Object Reference=@n10) -> (Output_Get:Texture 2D Object Reference)
   > Sync Browser!(execute, self, Object:Object Reference=@n12) -> (then) {is=title K2Node_CallFunction}
   > n13 = Increment Int(Value=Get Cur Texture Selected() -> (Cur Texture Selected=0))
   > n14 = Sequence!(execute) -> (then_0, then_1)
   n15 = Get Num Assets In This Collection() -> (Num Assets In This Collection:int=0)
   > Branch(Condition="integer > integer"(A=@n13, B=int - int(A=@n15)))
   > Set Cur Texture Selected(Cur Texture Selected=0) -> (Output_Get=0)
   n16 = Get Cur Texture()
   Set Cur Texture For Button Fixes(execute=@n14.then_1, Cur Texture For Button Fixes:Texture 2D Object Reference=@n16) -> (Output_Get:Texture 2D Object Reference)
   > RemoveFromParent(self=Get Actions Box())
   n17 = Find(TargetMap=Get All Textures(), Key=Get Cur Texture())
   > n18 = Update Health Tool Header!(
    execute
    Header Widget:BP UW Shared Header Widget Object Reference=Get Header()
    Cur Asset:Object Reference=@n16
    Cur Asset Error Struct:S Asset Errors Asset Health Structure=@n17
   ) -> (then, Resulting Buttons:Array of BP UW Shared Fix Me Button Object References)
   > Set ThisAssetButtons(ThisAssetButtons:Array of BP UW Shared Fix Me Button Object References=@n18) -> (Output_Get:Array of BP UW Shared Fix Me Button Object References)
   > AddActionsBox()
   > AddRunTestsButton()
   > AddNextTextureButton()
   > NextButtonEnable(Enable=true) {is=title K2Node_CallFunction}
   n19 = Break S Asset Errors Asset Health!(S_AssetErrors_AssetHealth:"S Asset Errors Asset Health Structure (by ref)"=@n17) -> (
    Errors_3_#1:Array of Texts
    Fixes_14_#2:Map of Strings to Strings
    ErrorTooltips_19_#3:Array of Texts
   ) {is=title K2Node_BreakStruct}
   > n20 = For Each Loop(Array:Array of Texts=@n19) -> (Array Element:Text)
   n21 = "To String (Text)"(InText=@n20)
   > Branch(Condition=Find(TargetMap:Map of Strings to Strings=@n19.Fixes_14_#2, Key:"String (by ref)"=@n21) -> (Value:String).ReturnValue)
   > n22 = Switch on String!(execute:Exec, Selection:String=@n21) -> (Default:Exec, Texture Too Large:Exec, Not power of 2:Exec) {pinNames [Texture Too Large, Not power of 2]}
   n23 = "Get (a copy)"(
    Array:Array of BP UW Shared Fix Me Button Object References=Get ThisAssetButtons() -> (ThisAssetButtons:Array of BP UW Shared Fix Me Button Object References)
    Dimension 1=@n20.Array Index
   ) -> (Output:BP UW Shared Fix Me Button Object Reference)
   Bind Event to Fix Me Button Pressed(execute=@n22.Not power of 2, self:BP UW Shared Fix Me Button Object Reference=@n23, Delegate=Create Event()) {is=title K2Node_AddDelegate}
   Bind Event to Fix Me Button Pressed(execute=@n22.Texture Too Large, self=@n23, Delegate=Create Event())
   DisplayUserWarningMessage(execute=@n11.else, Message="LOCTEXT(\"Click Run Tests first\")")
   > Return Node()
   n24 = Get Assets to Cycle Through(self=Get Settings()) -> (Assets to Cycle Through=NewEnumerator0)
   DisplayUserWarningMessage(execute=@n4.else, Message=Format Text(Format="LOCTEXT(\"No {s} assets found\")", s=Enum to String(Enumerator=@n24)))
   n25 = For Each Loop(Array:Array of Asset Data Structures=@n5) -> (Array Element:Asset Data Structure)
   n26 = SyncBrowserToObjects(execute=@n25.Completed, AssetPaths=Get CollectionAssetPaths() -> (CollectionAssetPaths:Array of Strings))
   Add(execute=@n25.LoopBody, TargetArray=Get CollectionAssetPaths(), NewItem=GetSoftObjectPath(Object=GetAsset(InAssetData=@n25))) {is=title K2Node_CallArrayFunction}
   GetFullName!(InAssetData:"Asset Data Structure (by ref)"=@n25) -> (ReturnValue:String)
  ]}
 On Run Tests Clicked {nodes [
   n1 = Get Project Asset Path() -> (Project Asset Path:String)
   n2 = GetAssetsByPath!(
    execute
    self:Asset Registry Interface=GetAssetRegistry()
    PackagePath:Name=String To Name(InString=@n1)
    bRecursive:bool=true
    bIncludeOnlyOnDiskAssets:bool=false
   ) -> (then, OutAssetData:Array of Asset Data Structures, ReturnValue:bool)
   > n3 = For Each Loop(Array=@n2)
   > Branch(Condition="Equal (Class)"(A=GetClass!(InAssetData:"Asset Data Structure (by ref)"=@n3) -> (ReturnValue:Object Class Reference), B=/Script/Engine.Texture2D))
   > n4 = Cast To Texture2D(Object=GetAsset(InAssetData=@n3))
   > Add(
    TargetArray:Array of Texture 2D Object References=Get Project Texture 2d() -> (Project Texture 2d:Array of Texture 2D Object References)
    NewItem:"Texture 2D Object Reference (by ref)"=@n4
   )
   On Run Tests Clicked() {is=title K2Node_FunctionEntry}
   > n5 = Sequence!(execute) -> (then_0, then_1, then_2:Exec, then_3:Exec, then_4:Exec, then_5:Exec, then_6:Exec, then_7:Exec, then_8:Exec)
   > DisplayUserWarningMessage()
   > Set FinalRunMessage()
   n6 = GetProjectAssetsWithFiltering!(execute=@n5.then_1, ProjPath:String=@n1, Filter:String=DO_NOT_OPTIMIZE) -> (then, Found Static Meshes:Array of Static Mesh Object References, Found Textures:Array of Texture 2D Object References)
   > Set Project Texture 2d(Project Texture 2d:Array of Texture 2D Object References=@n6.Found Textures) -> (Output_Get:Array of Texture 2D Object References)
   K2Node_Composite_0 = Run Checks(execute=@n5.then_2) {is=title K2Node_Composite}
   K2Node_Composite_2 = Sort Severity!(InputPin:Exec=@n5.then_3) {is=title K2Node_Composite}
   Set Cur Texture Selected(execute=@n5.then_4, Cur Texture Selected=0) -> (Output_Get=0)
   n7 = Branch(execute=@n5.then_8, Condition="integer > integer"(A=Get Num Severe() -> (Num Severe:int=0)))
   > n8 = Set Assets to Cycle Through!(
    execute
    Assets to Cycle Through:E_CollectionChoices_TextureHealth Enum=NewEnumerator0
    self:BP STPS Texture Health Object Reference=Get Settings()
   ) -> (then, Output_Get:E_CollectionChoices_TextureHealth Enum=NewEnumerator0)
   NextButtonEnable(execute=[@n16.then, @n14.then, @n8.then], Enable=true)
   > n9 = CreateCollectionMessage!(execute, CurSev:E_CollectionChoices_TextureHealth Enum=NewEnumerator0, NumFound:int=Get Num Severe() -> (Num Severe=0), CurrentMessage:Text) -> (then, Message:Text) {is=title K2Node_CallFunction}
   > n10 = CreateCollectionMessage(CurSev=NewEnumerator1, NumFound=Get Num Moderate() -> (Num Moderate:int=0), CurrentMessage=@n9)
   > n11 = CreateCollectionMessage(CurSev=NewEnumerator2, NumFound=Get Num Minor() -> (Num Minor:int=0), CurrentMessage=@n10)
   > n12 = Set FinalRunMessage(FinalRunMessage=@n11)
   > DisplayUserWarningMessage(Message=@n12)
   n13 = Branch(execute=@n7.else, Condition="integer > integer"(A=Get Num Moderate() -> (Num Moderate=0)))
   > n14 = Set Assets to Cycle Through(Assets to Cycle Through=NewEnumerator1, self=Get Settings()) -> (Output_Get=NewEnumerator0)
   n15 = Branch(execute=@n13.else, Condition="integer > integer"(A=Get Num Minor() -> (Num Minor=0)))
   > n16 = Set Assets to Cycle Through(Assets to Cycle Through=NewEnumerator2, self=Get Settings()) -> (Output_Get=NewEnumerator0)
   DisplayUserWarningMessage(execute=@n15.else, Message="LOCTEXT(\"No problems found!\")")
   n17 = Branch(execute=@n5.then_5, Condition="integer > integer"(A=Get Num Severe() -> (Num Severe=0)))
   > n18 = Make Collection!(
    execute
    Collection Name:String=Get Severe Collection Name()
    Parent Collection Name:String=Get Parent Collection Name()
    ShareType:ECollectionScriptingShareType Enum=Get Collection Share Type() -> (Collection Share Type=Local)
    Container:Collection Container Source Structure=Get Valkyrie Main Proj Collection Container()
    Parent Color:Linear Color Structure=Get Parent Color() -> (Parent Color:Linear Color Structure)
    Collection Color:Linear Color Structure=Get Severe Collection Color() -> (Severe Collection Color:Linear Color Structure)
   ) -> (then, OutNewOrEmptyCollection:Collection Structure, Success:bool, OutParentCollection:Collection Structure)
   > Set Severe Collection(Severe Collection:Collection Structure=@n18) -> (Output_Get:Collection Structure)
   > Set Parent Collection(Parent Collection:Collection Structure=@n18.OutParentCollection) -> (Output_Get:Collection Structure)
   > n19 = "For Each Loop (Map)"!(
    execute:Exec
    MapPin:Map of Texture 2D Object References to S Asset Errors Asset Health Structures=Get Textures Severe() -> (Textures Severe:Map of Texture 2D Object References to S Asset Errors Asset Health Structures)
    BreakPin:Exec
   ) -> (then:Exec, KeyPin:Texture 2D Object Reference, ValuePin:S Asset Errors Asset Health Structure, CompletedPin:Exec)
   > AddAssetToCollection!(
    execute
    self:Collection Manager Subsystem Object Reference=Get CollectionManagerScriptingSubsystem()
    Collection:"Collection Structure (by ref)"=Get Severe Collection()
    AssetPath:"Soft Object Path Structure (by ref)"=GetSoftObjectPath!(Object=@n19) -> (ReturnValue:Soft Object Path Structure)
   ) -> (then, ReturnValue:bool)
   n20 = CollectionExists(
    execute=@n17.else
    self=Get CollectionManagerScriptingSubsystem()
    Container=Get Valkyrie Main Proj Collection Container()
    Collection=String To Name(InString=Get Severe Collection Name())
    ShareType=Get Collection Share Type() -> (Collection Share Type=Local)
   )
   > Branch(Condition=@n20)
   > Destroy Collection!(
    execute
    Collection Name:String=Get Severe Collection Name()
    ShareType:ECollectionScriptingShareType Enum=Get Collection Share Type() -> (Collection Share Type=Local)
    CollectionScriptingContainerSource:"Collection Container Source Structure (by ref)"=Get Valkyrie Main Proj Collection Container()
   ) -> (then, Success:bool)
   n21 = Branch(execute=@n5.then_6, Condition="integer > integer"(A=Get Num Moderate() -> (Num Moderate=0)))
   > n22 = Make Collection(
    Collection Name=Get Moderate Collection Name()
    Parent Collection Name=Get Parent Collection Name()
    ShareType=Get Collection Share Type() -> (Collection Share Type=Local)
    Container=Get Valkyrie Main Proj Collection Container()
    Parent Color=Get Parent Color()
    Collection Color=Get Moderate Collection Color() -> (Moderate Collection Color:Linear Color Structure)
   )
   > Set Moderate Collection(Moderate Collection:Collection Structure=@n22) -> (Output_Get:Collection Structure)
   > Set Parent Collection(Parent Collection=@n22.OutParentCollection)
   > n23 = "For Each Loop (Map)"(MapPin=Get Textures Moderate() -> (Textures Moderate:Map of Texture 2D Object References to S Asset Errors Asset Health Structures))
   > AddAssetToCollection(self=Get CollectionManagerScriptingSubsystem(), Collection=Get Moderate Collection(), AssetPath=GetSoftObjectPath(Object=@n23))
   n24 = CollectionExists(
    execute=@n21.else
    self=Get CollectionManagerScriptingSubsystem()
    Container=Get Valkyrie Main Proj Collection Container()
    Collection=String To Name(InString=Get Moderate Collection Name())
    ShareType=Get Collection Share Type() -> (Collection Share Type=Local)
   )
   > Branch(Condition=@n24)
   > Destroy Collection(
    Collection Name=Get Moderate Collection Name()
    ShareType=Get Collection Share Type() -> (Collection Share Type=Local)
    CollectionScriptingContainerSource=Get Valkyrie Main Proj Collection Container()
   )
   n25 = Branch(execute=@n5.then_7, Condition="integer > integer"(A=Get Num Minor() -> (Num Minor=0)))
   > n26 = Make Collection(
    Collection Name=Get Minor Collection Name()
    Parent Collection Name=Get Parent Collection Name()
    ShareType=Get Collection Share Type() -> (Collection Share Type=Local)
    Container=Get Valkyrie Main Proj Collection Container()
    Parent Color=Get Parent Color()
    Collection Color=Get Minor Collection Color() -> (Minor Collection Color:Linear Color Structure)
   )
   > Set Minor Collection(Minor Collection:Collection Structure=@n26) -> (Output_Get:Collection Structure)
   > Set Parent Collection(Parent Collection=@n26.OutParentCollection)
   > n27 = "For Each Loop (Map)"(MapPin=Get Textures Minor() -> (Textures Minor:Map of Texture 2D Object References to S Asset Errors Asset Health Structures))
   > AddAssetToCollection(self=Get CollectionManagerScriptingSubsystem(), Collection=Get Minor Collection(), AssetPath=GetSoftObjectPath(Object=@n27))
   n28 = CollectionExists(
    execute=@n25.else
    self=Get CollectionManagerScriptingSubsystem()
    Container=Get Valkyrie Main Proj Collection Container()
    Collection=String To Name(InString=Get Minor Collection Name())
    ShareType=Get Collection Share Type() -> (Collection Share Type=Local)
   )
   > Branch(Condition=@n28)
   > Destroy Collection(
    Collection Name=Get Minor Collection Name()
    ShareType=Get Collection Share Type() -> (Collection Share Type=Local)
    CollectionScriptingContainerSource=Get Valkyrie Main Proj Collection Container()
   )
  ]}
 Run Checks {
  path=On Run Tests Clicked.K2Node_Composite_0.Run Checks
  nodes [
   Inputs()
   > n1 = Sequence!(execute) -> (then_0, then_1)
   > Clear!(
    execute
    TargetMap:Map of Texture 2D Object References to S Asset Errors Asset Health Structures=Get VectorDisplacementArray() -> (VectorDisplacementArray:Map of Texture 2D Object References to S Asset Errors Asset Health Structures)
   ) -> (then) {is=title K2Node_CallFunction}
   > Clear(TargetMap=Get CompressionGrayscaleArray() -> (CompressionGrayscaleArray:Map of Texture 2D Object References to S Asset Errors Asset Health Structures))
   > Clear(TargetMap=Get Textures Large() -> (Textures Large:Map of Texture 2D Object References to S Asset Errors Asset Health Structures))
   > Clear(TargetMap=Get UncompressedArray() -> (UncompressedArray:Map of Texture 2D Object References to S Asset Errors Asset Health Structures))
   > Clear(TargetMap=Get Textures Not Pow 2() -> (Textures Not Pow 2:Map of Texture 2D Object References to S Asset Errors Asset Health Structures))
   > Clear(TargetMap=Get Textures Not Streaming() -> (Textures Not Streaming:Map of Texture 2D Object References to S Asset Errors Asset Health Structures))
   n2 = For Each Loop(Exec=@n1.then_1, Array:Array of Texture 2D Object References=Get Project Texture 2d()) -> (Array Element:Texture 2D Object Reference)
   > Set Cur Project T2d(Cur Project T2d:Texture 2D Object Reference=@n2) -> (Output_Get:Texture 2D Object Reference)
   > n3 = Sequence!(execute) -> (then_0, then_1, then_2:Exec, then_3:Exec, then_4:Exec, then_5:Exec)
   n4 = TextureCheck_NotPow2!(
    execute=@n3.then_1
    Texture:Texture 2D Object Reference=Get Cur Project T2d() -> (Cur Project T2d:Texture 2D Object Reference)
    Not Pow 2 Array:Map of Texture 2D Object References to S Asset Errors Asset Health Structures=Get Textures Not Pow 2()
    Only Major:bool=true
    Major Texture Threshold:int=1024
   ) -> (then, Updated Not Pow 2 Array:Map of Texture 2D Object References to S Asset Errors Asset Health Structures)
   > Set Textures Not Pow 2(Textures Not Pow 2:Map of Texture 2D Object References to S Asset Errors Asset Health Structures=@n4) -> (Output_Get:Map of Texture 2D Object References to S Asset Errors Asset Health Structures)
   n5 = TextureCheck_NotStreaming!(
    execute=@n3.then_2
    Texture:Texture 2D Object Reference=Get Cur Project T2d()
    Not Streaming Array:Map of Texture 2D Object References to S Asset Errors Asset Health Structures=Get Textures Not Streaming()
    Only Major:bool=true
    Major Texture Threshold:int=1024
   ) -> (then, Updated Temp Not Streaming Array:Map of Texture 2D Object References to S Asset Errors Asset Health Structures)
   > Set Textures Not Streaming(Textures Not Streaming:Map of Texture 2D Object References to S Asset Errors Asset Health Structures=@n5) -> (Output_Get:Map of Texture 2D Object References to S Asset Errors Asset Health Structures)
   n6 = TextureCheck_Compression_GrayscaleSRGB!(
    execute=@n3.then_3
    Texture:Texture 2D Object Reference=Get Cur Project T2d()
    Texture Array:Map of Texture 2D Object References to S Asset Errors Asset Health Structures=Get CompressionGrayscaleArray()
    Only Major:bool=true
    Major Texture Threshold:int=1024
   ) -> (then, Is Ok Grayscale:bool, Updated Texture Array:Map of Texture 2D Object References to S Asset Errors Asset Health Structures)
   > Set CompressionGrayscaleArray(CompressionGrayscaleArray:Map of Texture 2D Object References to S Asset Errors Asset Health Structures=@n6.Updated Texture Array) -> (Output_Get:Map of Texture 2D Object References to S Asset Errors Asset Health Structures)
   n7 = TextureCheck_Compression_Uncompressed!(
    execute=@n3.then_4
    Texture:Texture 2D Object Reference=Get Cur Project T2d()
    Texture Array:Map of Texture 2D Object References to S Asset Errors Asset Health Structures=Get UncompressedArray()
    Only Major:bool=true
    Major Texture Threshold:int=1024
   ) -> (then, Is Ok Compression:bool, Updated Texture Array:Map of Texture 2D Object References to S Asset Errors Asset Health Structures)
   > Set UncompressedArray(UncompressedArray:Map of Texture 2D Object References to S Asset Errors Asset Health Structures=@n7.Updated Texture Array) -> (Output_Get:Map of Texture 2D Object References to S Asset Errors Asset Health Structures)
   n8 = TextureCheck_Compression_VectorDisplacement!(
    execute=@n3.then_5
    Texture:Texture 2D Object Reference=Get Cur Project T2d()
    Texture Array:Map of Texture 2D Object References to S Asset Errors Asset Health Structures=Get VectorDisplacementArray()
    Only Major:bool=true
    Major Texture Threshold:int=1024
   ) -> (then, Is Ok Compression:bool, Updated Texture Array:Map of Texture 2D Object References to S Asset Errors Asset Health Structures)
   > Set VectorDisplacementArray(VectorDisplacementArray:Map of Texture 2D Object References to S Asset Errors Asset Health Structures=@n8.Updated Texture Array) -> (Output_Get:Map of Texture 2D Object References to S Asset Errors Asset Health Structures)
   n9 = TextureCheck_IsTooLarge!(
    execute=@n3.then_0
    Texture:Texture 2D Object Reference=Get Cur Project T2d()
    Size Threshold:int=Get Texture Size Threshold() -> (Texture Size Threshold:int=0)
    Too Large Array:Map of Texture 2D Object References to S Asset Errors Asset Health Structures=Get Textures Large()
   ) -> (then, Updated Array:Map of Texture 2D Object References to S Asset Errors Asset Health Structures)
   > Set Textures Large(Textures Large:Map of Texture 2D Object References to S Asset Errors Asset Health Structures=@n9) -> (Output_Get:Map of Texture 2D Object References to S Asset Errors Asset Health Structures)
   Clear(execute=@n2.Completed, TargetMap=Get All Textures())
   > n10 = Sequence!(execute) -> (then_0, then_1, then_2)
   > n11 = Texture Map Merge!(
    execute
    self
    Array 1:Map of Texture 2D Object References to S Asset Errors Asset Health Structures=Get Textures Not Streaming()
    Merge Dest Array:Map of Texture 2D Object References to S Asset Errors Asset Health Structures=Get All Textures()
   ) -> (then, Merged:Map of Texture 2D Object References to S Asset Errors Asset Health Structures)
   > Set All Textures(All Textures:Map of Texture 2D Object References to S Asset Errors Asset Health Structures=@n11) -> (Output_Get:Map of Texture 2D Object References to S Asset Errors Asset Health Structures)
   > n12 = Texture Map Merge(Array 1=Get Textures Large(), Merge Dest Array=Get All Textures())
   > Set All Textures(All Textures=@n12)
   > n13 = Texture Map Merge(Array 1=Get Textures Not Pow 2(), Merge Dest Array=Get All Textures())
   > Set All Textures(All Textures=@n13)
   n14 = Texture Map Merge(execute=@n10.then_1, Array 1=Get CompressionGrayscaleArray(), Merge Dest Array=Get All Textures())
   > Set All Textures(All Textures=@n14)
   > n15 = Texture Map Merge(Array 1=Get UncompressedArray(), Merge Dest Array=Get All Textures())
   > Set All Textures(All Textures=@n15)
   > n16 = Texture Map Merge(Array 1=Get VectorDisplacementArray(), Merge Dest Array=Get All Textures())
   > Set All Textures(All Textures=@n16)
   Outputs()
  ]
 }
 Sort Severity {
  path=On Run Tests Clicked.K2Node_Composite_2.Sort Severity
  nodes [
   Inputs!() -> (InputPin:Exec)
   > Sequence!(execute) -> (then_0, then_1)
   > Clear(TargetMap=Get Textures Severe())
   > Clear(TargetMap=Get Textures Moderate())
   > Clear(TargetMap=Get Textures Minor())
   > Set Num Severe(Num Severe:int=0) -> (Output_Get:int=0)
   > Set Num Moderate(Num Moderate:int=0) -> (Output_Get:int=0)
   > Set Num Minor(Num Minor:int=0) -> (Output_Get:int=0)
   > n1 = "For Each Loop (Map)"(MapPin=Get All Textures())
   n2 = Length(TargetArray:Array of Texts=Break S Asset Errors Asset Health(S_AssetErrors_AssetHealth=@n1.ValuePin))
   > n3 = Branch(Condition="integer > integer"(A=@n2, B=2))
   > Increment Int(Value=Get Num Severe() -> (Num Severe=0))
   > Add!(
    execute
    TargetMap:Map of Texture 2D Object References to S Asset Errors Asset Health Structures=Get Textures Severe()
    Key:"Texture 2D Object Reference (by ref)"=@n1
    Value:"S Asset Errors Asset Health Structure (by ref)"=@n1.ValuePin
   ) -> (then) {is=title K2Node_CallFunction}
   n4 = Branch(execute=@n3.else, Condition="Equal (Integer)"(A=@n2, B=2))
   > Increment Int(Value=Get Num Moderate() -> (Num Moderate=0))
   > Add(TargetMap=Get Textures Moderate(), Key=@n1, Value=@n1.ValuePin)
   Increment Int(" "=@n4.else, Value=Get Num Minor() -> (Num Minor=0))
   > Add(TargetMap=Get Textures Minor(), Key=@n1, Value=@n1.ValuePin)
   Make Map!(Key 0:Texture 2D Object Reference=@n1, Value 0:S Asset Errors Asset Health Structure=@n1.ValuePin) -> (Map:Map of Texture 2D Object References to S Asset Errors Asset Health Structures) {is=title K2Node_MakeMap}
   Outputs()
  ]
 }
 On Severity Changed {nodes [
   On Severity Changed!() -> (then, PropertySet:Scriptable Interactive Tool Property Set Object Reference, PropertyName:String) {is=title K2Node_FunctionEntry}
   > Set Cur Texture Selected(Cur Texture Selected=0) -> (Output_Get=0)
  ]}
 OnClearCollectionClicked {nodes [
   OnClearCollectionClicked() {is=title K2Node_FunctionEntry}
   > n1 = GetCollections!(
    execute
    self:Collection Manager Subsystem Object Reference=Get CollectionManagerScriptingSubsystem()
    Container:Collection Container Source Structure=Get Valkyrie Main Proj Collection Container()
   ) -> (then, OutCollections:Array of Collection Structures, ReturnValue:bool)
   > n2 = For Each Loop(Array:Array of Collection Structures=Make Array!(
     "[0]":Collection Structure=Get Severe Collection()
     "[1]":Collection Structure=Get Moderate Collection()
     "[2]":Collection Structure=Get Minor Collection()
     "[3]":Collection Structure=Get Parent Collection() -> (Parent Collection:Collection Structure)
    ) -> (Array:Array of Collection Structures)) -> (Array Element:Collection Structure)
   > Branch(Condition="integer > integer"(A=Find Item!(TargetArray:Array of Collection Structures=@n1, ItemToFind:"Collection Structure (by ref)"=@n2) -> (ReturnValue), B=-1))
   > DestroyCollection!(
    execute
    self:Collection Manager Subsystem Object Reference=Get CollectionManagerScriptingSubsystem()
    Collection:"Collection Structure (by ref)"=@n2
   ) -> (then, ReturnValue:bool)
  ]}
 Populate Collection Vars {nodes [
   Populate Collection Vars() {is=title K2Node_FunctionEntry}
   > n1 = GetCollections(self=Get CollectionManagerScriptingSubsystem(), Container=Get Valkyrie Main Proj Collection Container())
   > n2 = For Each Loop(Array=@n1)
   > n3 = Set Cur Found Array(Cur Found Array:Collection Structure=@n2) -> (Output_Get:Collection Structure)
   n4 = Break Collection(CollectionScriptingRef=@n3)
   n5 = "To String (Name)"(InName=@n4.Name)
   n6 = Get Parent Collection Name()
   n7 = Get Collection Share Type() -> (Collection Share Type=Local)
   > n8 = Branch(Condition=AND Boolean(
     A="Equal Exactly (String)"(A=@n6, B=@n5)
     B="Equal (Enum)"(A:ECollectionScriptingShareType Enum=@n4.ShareType, B:ECollectionScriptingShareType Enum=@n7)
    ))
   > Set Parent Collection(Parent Collection=Get Cur Found Array() -> (Cur Found Array:Collection Structure))
   n9 = Break Collection(CollectionScriptingRef=Get Cur Found Array())
   n10 = "To String (Name)"(InName=@n9.Name)
   n11 = Get Severe Collection Name()
   n12 = Get Collection Share Type() -> (Collection Share Type=Local)
   n13 = Branch(execute=@n8.else, Condition=AND Boolean(A="Equal Exactly (String)"(A=@n11, B=@n10), B="Equal (Enum)"(A=@n9.ShareType, B=@n12)))
   > Set Severe Collection(Severe Collection=Get Cur Found Array())
   n14 = Break Collection(CollectionScriptingRef=Get Cur Found Array())
   n15 = "To String (Name)"(InName=@n14.Name)
   n16 = Get Moderate Collection Name()
   n17 = Get Collection Share Type() -> (Collection Share Type=Local)
   n18 = Branch(execute=@n13.else, Condition=AND Boolean(A="Equal Exactly (String)"(A=@n16, B=@n15), B="Equal (Enum)"(A=@n14.ShareType, B=@n17)))
   > Set Moderate Collection(Moderate Collection=Get Cur Found Array())
   n19 = Break Collection(CollectionScriptingRef=Get Cur Found Array())
   n20 = "To String (Name)"(InName=@n19.Name)
   n21 = Get Minor Collection Name()
   n22 = Get Collection Share Type() -> (Collection Share Type=Local)
   Branch(execute=@n18.else, Condition=AND Boolean(A="Equal Exactly (String)"(A=@n21, B=@n20), B="Equal (Enum)"(A=@n19.ShareType, B=@n22)))
   > Set Minor Collection(Minor Collection=Get Cur Found Array())
  ]}
 Run Pow2 Fix {nodes [
   Run Pow2 Fix() {is=title K2Node_FunctionEntry}
   > TextureFix_PowerOf2!(
    execute
    Texture:Texture 2D Object Reference=Get Cur Texture For Button Fixes() -> (Cur Texture For Button Fixes:Texture 2D Object Reference)
   ) -> (then, Fix Successful:bool)
   > CloseOpenAssetEditor!(
    execute
    Asset:Object Reference=Get Cur Texture For Button Fixes()
    Close:bool=true
    Open:bool=true
    OpenedMethod:EAssetTypeActivationOpenedMethod Enum=Edit
   ) -> (then, Num Editors Closed:int=0, Open Succes:bool)
  ]}
 Run Too Large Fix {nodes [
   Run Too Large Fix() {is=title K2Node_FunctionEntry}
   > TextureFix_MaxSize!(execute, Texture:Texture 2D Object Reference=Get Cur Texture For Button Fixes(), Maximum Texture Size:int=2048) -> (then, Fix Successful:bool)
  ]}
 Texture Map Merge {nodes [
   n1 = Texture Map Merge!() -> (
    then
    Array 1:Map of Texture 2D Object References to S Asset Errors Asset Health Structures
    Merge Dest Array:Map of Texture 2D Object References to S Asset Errors Asset Health Structures
   ) {is=title K2Node_FunctionEntry}
   > Set Temp Merged(Temp Merged:Map of Texture 2D Object References to S Asset Errors Asset Health Structures=@n1.Merge Dest Array) -> (Output_Get:Map of Texture 2D Object References to S Asset Errors Asset Health Structures)
   > n2 = "For Each Loop (Map)"(MapPin=Get Array 1() -> (Array 1:Map of Texture 2D Object References to S Asset Errors Asset Health Structures))
   > Set Current Texture(Current Texture:Texture 2D Object Reference=@n2) -> (Output_Get:Texture 2D Object Reference)
   > Set Array 1 Value(Array 1 Value:S Asset Errors Asset Health Structure=Find(
     TargetMap:Map of Texture 2D Object References to S Asset Errors Asset Health Structures=Get Array 1()
     Key:"Texture 2D Object Reference (by ref)"=Get Current Texture() -> (Current Texture:Texture 2D Object Reference)
    ) -> (Value:S Asset Errors Asset Health Structure)) -> (Output_Get:S Asset Errors Asset Health Structure)
   n3 = Find(
    TargetMap=Get Temp Merged() -> (Temp Merged:Map of Texture 2D Object References to S Asset Errors Asset Health Structures)
    Key=Get Current Texture()
   )
   > Set Found In Temp Merged(Found In Temp Merged:S Asset Errors Asset Health Structure=@n3) -> (Output_Get:S Asset Errors Asset Health Structure)
   > n4 = Branch(Condition=@n3.ReturnValue)
   n5 = Break S Asset Errors Asset Health(S_AssetErrors_AssetHealth=Find(TargetMap=Get Temp Merged(), Key=Get Current Texture()))
   > Set Temp Merged Messages(Temp Merged Messages:Array of Texts=@n5) -> (Output_Get:Array of Texts)
   > Set Temp Merged Fixes(Temp Merged Fixes:Map of Strings to Strings=@n5.Fixes_14_#2) -> (Output_Get:Map of Strings to Strings)
   > Set TempMergedTooltips(TempMergedTooltips:Array of Texts=@n5.ErrorTooltips_19_#3) -> (Output_Get:Array of Texts)
   n6 = Break S Asset Errors Asset Health(S_AssetErrors_AssetHealth=Get Array 1 Value() -> (Array 1 Value:S Asset Errors Asset Health Structure))
   > Set Temp Array 1 Fixes(Temp Array 1 Fixes:Map of Strings to Strings=@n6.Fixes_14_#2) -> (Output_Get:Map of Strings to Strings)
   > Set Array 1 Tooltips(Array 1 Tooltips:Array of Texts=@n6.ErrorTooltips_19_#3) -> (Output_Get:Array of Texts)
   > n7 = For Each Loop(Array:Array of Texts=@n6) -> (Array Element:Text)
   n8 = For Each Loop(Exec=@n7.Completed, Array=Get Temp Merged Messages() -> (Temp Merged Messages:Array of Texts))
   Remove!(
    execute=@n8.Completed
    TargetMap:Map of Texture 2D Object References to S Asset Errors Asset Health Structures=Get Temp Merged()
    Key:"Texture 2D Object Reference (by ref)"=Get Current Texture()
   ) -> (then, ReturnValue:bool)
   > Add(
    TargetMap=Get Temp Merged()
    Key=Get Current Texture()
    Value=Make S Asset Errors Asset Health!(
     Errors_3_#1:Array of Texts=Get Temp Merged Messages()
     Fixes_14_#2:Map of Strings to Strings=Get Temp New Fixes() -> (Temp New Fixes:Map of Strings to Strings)
     ErrorTooltips_19_#3:Array of Texts=Get TempMergedTooltips() -> (TempMergedTooltips:Array of Texts)
    ) -> (S_AssetErrors_AssetHealth:S Asset Errors Asset Health Structure) {is=title K2Node_MakeStruct}
   )
   n9 = "To String (Text)"(InText=@n8)
   n10 = Find(TargetMap:Map of Strings to Strings=Get Temp Array 1 Fixes() -> (Temp Array 1 Fixes:Map of Strings to Strings), Key:"String (by ref)"=@n9) -> (Value:String)
   n11 = Branch(execute=@n8.LoopBody, Condition=@n10.ReturnValue)
   > Add(TargetMap:Map of Strings to Strings=Get Temp New Fixes(), Key:"String (by ref)"=@n9, Value:"String (by ref)"=@n10)
   n12 = Find(TargetMap=Get Temp Merged Fixes() -> (Temp Merged Fixes:Map of Strings to Strings), Key=@n9)
   Branch(execute=@n11.else, Condition=@n12.ReturnValue)
   > Add(TargetMap=Get Temp New Fixes(), Key=@n9, Value=@n12)
   Add Unique(execute=@n7.LoopBody, TargetArray:Array of Texts=Get Temp Merged Messages(), NewItem:"Text (by ref)"=@n7)
   > Add(
    TargetArray:Array of Texts=Get TempMergedTooltips()
    NewItem:"Text (by ref)"="Get (a ref)"(Array:Array of Texts=Get Array 1 Tooltips() -> (Array 1 Tooltips:Array of Texts), Dimension 1=@n7.Array Index) -> (Output:"Text (by ref)")
   ) {is=title K2Node_CallArrayFunction}
   Add(
    execute=@n4.else
    TargetMap:Map of Texture 2D Object References to S Asset Errors Asset Health Structures=Get Temp Merged()
    Key:"Texture 2D Object Reference (by ref)"=Get Current Texture()
    Value:"S Asset Errors Asset Health Structure (by ref)"=Get Array 1 Value()
   ) {is=title K2Node_CallFunction}
   Return Node!(execute=@n2.CompletedPin, Merged:Map of Texture 2D Object References to S Asset Errors Asset Health Structures=Get Temp Merged())
  ]}
]
variables [
 Actions Box {type=BP_UW_ExpandableBox_C}
 All Textures {}
 Collection Share Type {type=ECollectionScriptingShareType, default=Local}
 CompressionGrayscaleArray {}
 Cur Texture For Button Fixes {type=Texture2D}
 Cur Texture Selected {type=integer}
 Currently Selected Asset {type=Texture2D}
 CurTime {type=number}
 FinalRunMessage {type=string, default=""}
 FirstInitRun {type=boolean}
 Header {type=BP_UW_SharedHeaderWidget_C}
 Main Project Handle {type=ValkyrieProjectHandle, default {}}
 Minor Collection {type=CollectionScriptingRef, default {name=Minor, container=Game, shareType=Local}}
 Minor Collection Color {type=LinearColor, default=<rgba 0 0 0 0>}
 Minor Collection Name {type=string, default=Minor}
 Moderate Collection {type=CollectionScriptingRef, default {name=Moderate, container=Game, shareType=Local}}
 Moderate Collection Color {type=LinearColor, default=<rgba 0 0 0 0>}
 Moderate Collection Name {type=string, default=Moderate}
 NextButtonTooltip_Disabled {type=string, default=Run Tests to cycle to next texture.}
 NextButtonTooltip_Enabled {type=string, default=Cycles to the next texture in the collection.}
 NextTextureButton {type=BP_UW_NextTextureButton_C, default=null}
 Num Minor {type=integer}
 Num Moderate {type=integer}
 Num Severe {type=integer}
 Parent Collection {type=CollectionScriptingRef, default {name=None, container=Game, shareType=Local}}
 Parent Collection Name {type=string, default=Texture_Health}
 Parent Color {type=LinearColor, default=<rgba 0 0 0 0>}
 Previously Selected Asset {type=Texture2D}
 Project Asset Path {type=string, default=""}
 Project Handle Valid {type=boolean}
 Prop Set Save Key {type=string, default=SaveSettings}
 RunTestsButton {type=BP_UW_RunTestsButton_C, default=null}
 Settings {type=BP_STPS_TextureHealth_C, default=null}
 Severe Collection {type=CollectionScriptingRef, default {name=Severe, container=Game, shareType=Local}}
 Severe Collection Color {type=LinearColor, default=<rgba 0 0 0 0>}
 Severe Collection Name {type=string, default=Severe}
 Texture Size Threshold {type=integer, default=2048}
 Textures Large {}
 Textures Minor {}
 Textures Moderate {}
 Textures Not Pow 2 {}
 Textures Not Streaming {}
 Textures Severe {}
 UncompressedArray {}
 Valkyrie Main Proj Collection Container {type=CollectionScriptingContainerSource, default {name=Game, title=Collections}}
 VectorDisplacementArray {}
]
